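# Niagara system for the LemonCart Granite emote trail: a blue-violet whirlwind of DollarDollarBillYall sprites, smoke and wind meshes.
# Sprites spawn off a skeletal mesh and swirl under vortex and lift forces, facing radially.
# Smoke puffs and a ribbon follow a sine/cosine rising spiral driven by shared System variables, with CPU and GPU variants of sprites and puffs.
# WindBase and WindFunnel add fading meshes, the base with a light; DEBUGGhostMtl is disabled.
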NiagaraSystem /BRCosmetics/Effects/Fort_Effects/Emotes/LemonCart_Granite/NS_Emote_LemonCart_Granite_Trail
dependencies {
 usedDataInterfaces [&NiagaraDataInterfaceCurve, &NiagaraDataInterfaceSkeletalMesh]
 usedDynamicInputs []
 usedModules [
  &self:SystemSpawnScript.SystemScriptSource.SystemScriptGraph.NiagaraNodeAssignment_4.SetVariables_#1
  &self:SystemSpawnScript.SystemScriptSource.SystemScriptGraph.NiagaraNodeAssignment_1.SetVariables_#2
  &self:SystemSpawnScript.SystemScriptSource.SystemScriptGraph.NiagaraNodeAssignment_0.SetVariables_#3
  &self:SystemSpawnScript.SystemScriptSource.SystemScriptGraph.NiagaraNodeAssignment_2.SetVariables_#4
  &self:SystemSpawnScript.SystemScriptSource.SystemScriptGraph.NiagaraNodeAssignment_3.SetVariables_#5
  &SystemState
 ]
}
emitters {
 DEBUGGhostMtl {
  bEnabled=false
  emitterUpdateScript [
   &EmitterState(Life Cycle Mode=::NewEnumerator1, Loop Behavior=::NewEnumerator1, Loop Duration Mode=::NewEnumerator1, UseLoopDelay=false)
   &SpawnBurst_Instantaneous(Spawn Count=5, Use Spawn Probability=false, UseLoopCountLimit=false)
  ]
  particleSpawnScript [
   &InitializeParticle(Color=<rgba 0.0865 0.3008 1 1>, Lifetime=3, Mesh Scale Mode=::NewEnumerator3, UsePositionOffset=false, Write Material Random=true)
   &ShapeLocation(MaskSpawnGroup=false, Override Seed=false, Sphere Radius=69.3223, UseOwnerScale=true)
   &InitialMeshOrientation(Mesh Axis To Orient=<0 1 0>, Mesh Orientation Mode=::NewEnumerator2, Use Rotation Vector=false)
  ]
  particleUpdateScript [
   &ParticleState(Kill Particles When Lifetime Has Elapsed=false)
   &DynamicMaterialParameters(
    Index 0 Param 2=&FloatFromCurve(FloatCurve=DataInterface() {propertyValues {Curve {keys [
         {interpMode=RCIM_Cubic, tangentMode=RCTM_User, time=0, value=0.05}
         {arriveTangent=0.8093, interpMode=RCIM_Cubic, leaveTangent=0.8093, tangentMode=RCTM_User, time=1, value=0.3}
        ]}}})
    Index 0 Param 3=&FloatFromCurve(FloatCurve=DataInterface() {propertyValues {Curve {keys [
         {interpMode=RCIM_Cubic, tangentMode=RCTM_User, time=0, value=0}
         {arriveTangent=3.2342, interpMode=RCIM_Cubic, leaveTangent=3.2342, tangentMode=RCTM_User, time=1, value=1}
        ]}}})
    Index 0 Param 4=&RandomRangeFloat(Minimum=0, Maximum=1)
    Module.Index0Param0Write=true
    Module.Index0Param1Write=true
    Module.Index0Param2Write=true
    Module.Index0Param3Write=true
    Module.Index1Param0Write=true
    Module.Index1Param1Write=true
    Module.Index1Param2Write=true
    Module.Index1Param3Write=true
    Module.Index2Param0Write=true
    Module.Index2Param1Write=true
    Module.Index2Param2Write=true
    Module.Index2Param3Write=true
    Module.Index3Param0Write=true
    Module.Index3Param1Write=true
    Module.Index3Param2Write=true
    Module.Index3Param3Write=true
    Write Parameter Index 1=true
    Write Parameter Index 2=true
    Write Parameter Index 3=true
   )
  ]
  rendererClasses [&NiagaraSpriteRendererProperties, &NiagaraMeshRendererProperties]
  renderers [
   {rendererClass=&NiagaraSpriteRendererProperties}
   {rendererClass=&NiagaraMeshRendererProperties, rendererIndex=1}
  ]
 }
 DollarDollarBillYall {
  emitterUpdateScript [
   &EmitterState(Life Cycle Mode=::NewEnumerator1, Loop Behavior=::NewEnumerator1, Loop Duration=1.75, UseLoopDelay=false)
   &SpawnRate(SpawnRate=4, Use Spawn Probability=false)
   &SpawnBurst_Instantaneous(Spawn Count=200, Use Spawn Probability=false, UseLoopCountLimit=false) {enabled=false}
  ]
  particleSpawnScript [
   &InitializeParticle(
    Color Maximum=<rgba 0.2193 0.1719 1 1>
    Color Minimum=<rgba 0.0865 0.3008 1 1>
    Color Mode=::NewEnumerator2
    Lifetime Max=3
    Lifetime Min=0.5
    Lifetime Mode=::NewEnumerator1
    Mass Max=2
    Mass Min=1
    Mass Mode=::NewEnumerator3
    Mesh Scale Mode=::NewEnumerator3
    Sprite Rotation Angle=&Add_Float(A=-25, B=&RandomRangeFloat(Minimum=10, Maximum=-10))
    Sprite Rotation Mode=::NewEnumerator0
    Sprite Size Mode=::NewEnumerator1
    Uniform Sprite Size Max=1
    Uniform Sprite Size Min=0.5
    UsePositionOffset=false
    Write Material Random=true
   )
   &SkeletalMeshLocation(
    MaskSpawnGroup=false
    Mesh Sampling Type=::NewEnumerator5
    Orientation Sampling=::NewEnumerator1
    Position Sampling=::NewEnumerator2
    Skeletal Mesh=$System.Skeletal Mesh
    Socket Sampling Mode=::NewEnumerator5
   )
   &ShapeLocation(
    Handle Radius=5
    Large Radius=10
    MaskSpawnGroup=false
    Non Uniform Scale=<1 1 0.5>
    Override Seed=false
    Shape Origin=&AddVectorToPosition(Position=$StackContext.Position, Vector=<0 0 0>)
    Shape Primitive=::NewEnumerator3
    UseOwnerScale=true
   )
  ]
  particleUpdateScript [
   &ParticleState()
   /SetVariables(Particles.Lookat:Vector3f=&NormalizeVector(
     Vector To Normalize=&Subtract_Vector(
      A=&Multiply_Vector(A=Linked() {linkedVariable {name=Particles.Position, type=NiagaraPosition}}, B=<1 1 0>)
      B=&Multiply_Vector(A=Linked() {linkedVariable {name=Engine.Emitter.SimulationPosition, type=NiagaraPosition}}, B=<1 1 0>)
     )
     Safe Normalize=true
     Safe Fallback Threshold=0
     Safe Fallback Vector=<0 0 0.001>
    )) {isSetParameters=true, module=SetVariables_#6}
   &LinearForce(Force=<0 0 120>)
   &VortexForce(Origin Pull Amount=3000, UseOriginOffset=false, Vortex Force Amount=-900)
   &Drag(Use Linear Drag=true, Use Rotational Drag=true)
   &SolveForcesAndVelocity(Clamp Velocity=false, Limit Acceleration=false)
   &ScaleSpriteSize(
    Uniform Curve Scale=25
    Uniform Curve Sprite Scale=DataInterface() {propertyValues {Curve {keys [
        {interpMode=RCIM_Cubic, tangentMode=RCTM_User, time=0.5, value=1}
        {arriveTangent=21.7677, interpMode=RCIM_Cubic, leaveTangent=21.7677, tangentMode=RCTM_User, time=1, value=5}
       ]}}}
   )
   &ScaleSpriteSizeBySpeed(Max Scale Factor=<10 1>, Velocity Threshold=2000)
   &SpriteFacingAndAlignment(
    Sprite Alignment=&MakeVector(X=0, Y=0, Z=0)
    Sprite Facing=$Particles.Lookat
    WriteAlignment=false
    WriteFacing=true
   )
   &DynamicMaterialParameters(
    Index 0 Param 2=&FloatFromCurve(
     FloatCurve=DataInterface() {propertyValues {Curve {keys [
         {interpMode=RCIM_Cubic, tangentMode=RCTM_User, time=0.5, value=0}
         {arriveTangent=6.0243, interpMode=RCIM_Cubic, leaveTangent=6.0243, tangentMode=RCTM_User, time=1, value=1}
        ]}}}
     Scale Curve=2
    )
    Index 0 Param 3=&FloatFromCurve(
     FloatCurve=DataInterface() {propertyValues {Curve {keys [
         {interpMode=RCIM_Cubic, tangentMode=RCTM_User, time=0.5, value=0}
         {arriveTangent=4.3588, interpMode=RCIM_Cubic, leaveTangent=4.3588, tangentMode=RCTM_User, time=1, value=1}
        ]}}}
     Scale Curve=0.3
    )
    Index 0 Param 4=&RandomRangeFloat(Minimum=0, Maximum=1)
    Index 1 Param 2=1
    Index 1 Param 3=&RandomRangeFloat(Minimum=0.1, Maximum=1)
    Module.Index0Param0Write=true
    Module.Index0Param1Write=true
    Module.Index0Param2Write=true
    Module.Index0Param3Write=true
    Module.Index1Param0Write=true
    Module.Index1Param1Write=true
    Module.Index1Param2Write=true
    Module.Index1Param3Write=false
    Module.Index2Param0Write=true
    Module.Index2Param1Write=true
    Module.Index2Param2Write=true
    Module.Index2Param3Write=true
    Module.Index3Param0Write=true
    Module.Index3Param1Write=true
    Module.Index3Param2Write=true
    Module.Index3Param3Write=true
    Write Parameter Index 1=true
    Write Parameter Index 2=true
    Write Parameter Index 3=true
   )
   &ScaleColor(
    Scale Alpha=&FloatFromCurve(
     FloatCurve=DataInterface() {propertyValues {Curve {keys [
         {arriveTangent=9.0725, interpMode=RCIM_Cubic, leaveTangent=9.0725, tangentMode=RCTM_User, time=0, value=0}
         {arriveTangent=0.0337, interpMode=RCIM_Cubic, leaveTangent=0.0337, tangentMode=RCTM_User, time=0.1408, value=0.9847}
         {arriveTangent=0.0046, interpMode=RCIM_Cubic, leaveTangent=0.0046, tangentMode=RCTM_User, time=0.75, value=1}
         {arriveTangent=-0.0405, interpMode=RCIM_Cubic, leaveTangent=-0.0405, tangentMode=RCTM_User, time=0.9, value=0}
        ]}}}
     Scale Curve=&RandomRangeFloat(Minimum=0.25, Maximum=1)
    )
    Scale RGB=&VectorFromFloat(Value=&RandomRangeFloat(Minimum=1, Maximum=20))
    ScaleA=true
    ScaleRGB=false
   )
  ]
  rendererClasses [&NiagaraSpriteRendererProperties]
  renderers [{rendererClass=&NiagaraSpriteRendererProperties}]
 }
 DollarDollarBillYallGPU {
  simTarget=GPUComputeSim
  emitterUpdateScript [
   &EmitterState(Life Cycle Mode=::NewEnumerator1, Loop Behavior=::NewEnumerator1, Loop Duration=1.75, UseLoopDelay=false)
   &SpawnRate(SpawnRate=12, Use Spawn Probability=false)
   &SpawnBurst_Instantaneous(Spawn Count=200, Use Spawn Probability=false, UseLoopCountLimit=false) {enabled=false}
  ]
  particleSpawnScript [
   &InitializeParticle(
    Color Maximum=<rgba 0.2193 0.1719 1 1>
    Color Minimum=<rgba 0.0865 0.3008 1 1>
    Color Mode=::NewEnumerator2
    Lifetime Max=3
    Lifetime Min=0.5
    Lifetime Mode=::NewEnumerator1
    Mass Max=2
    Mass Min=1
    Mass Mode=::NewEnumerator3
    Mesh Scale Mode=::NewEnumerator3
    Sprite Rotation Angle=&Add_Float(A=-25, B=&RandomRangeFloat(Minimum=10, Maximum=-10))
    Sprite Rotation Mode=::NewEnumerator0
    Sprite Size Mode=::NewEnumerator1
    Uniform Sprite Size Max=1
    Uniform Sprite Size Min=0.5
    UsePositionOffset=false
    Write Material Random=true
   )
   &SkeletalMeshLocation(
    MaskSpawnGroup=false
    Mesh Sampling Type=::NewEnumerator5
    Orientation Sampling=::NewEnumerator1
    Position Sampling=::NewEnumerator2
    Skeletal Mesh=$System.Skeletal Mesh
    Socket Sampling Mode=::NewEnumerator5
   )
   &ShapeLocation(
    Handle Radius=5
    Large Radius=10
    MaskSpawnGroup=false
    Non Uniform Scale=<1 1 0.5>
    Override Seed=false
    Shape Origin=&AddVectorToPosition(Position=$StackContext.Position, Vector=<0 0 0>)
    Shape Primitive=::NewEnumerator3
    UseOwnerScale=true
   )
  ]
  particleUpdateScript [
   &ParticleState()
   /SetVariables(Particles.Lookat:Vector3f=&NormalizeVector(
     Vector To Normalize=&Subtract_Vector(
      A=&Multiply_Vector(A=Linked() {linkedVariable {name=Particles.Position, type=NiagaraPosition}}, B=<1 1 0>)
      B=&Multiply_Vector(A=Linked() {linkedVariable {name=Engine.Emitter.SimulationPosition, type=NiagaraPosition}}, B=<1 1 0>)
     )
     Safe Normalize=true
     Safe Fallback Threshold=0
     Safe Fallback Vector=<0 0 0.001>
    )) {isSetParameters=true, module=SetVariables_#6}
   &LinearForce(Force=<0 0 120>)
   &VortexForce(Origin Pull Amount=3000, UseOriginOffset=false, Vortex Force Amount=-900)
   &Drag(Use Linear Drag=true, Use Rotational Drag=true)
   &SolveForcesAndVelocity(Clamp Velocity=false, Limit Acceleration=false)
   &ScaleSpriteSize(
    Uniform Curve Scale=25
    Uniform Curve Sprite Scale=DataInterface() {propertyValues {Curve {keys [
        {interpMode=RCIM_Cubic, tangentMode=RCTM_User, time=0.5, value=1}
        {arriveTangent=21.7677, interpMode=RCIM_Cubic, leaveTangent=21.7677, tangentMode=RCTM_User, time=1, value=5}
       ]}}}
   )
   &ScaleSpriteSizeBySpeed(Max Scale Factor=<10 1>, Velocity Threshold=2000)
   &SpriteFacingAndAlignment(
    Sprite Alignment=&MakeVector(X=0, Y=0, Z=0)
    Sprite Facing=$Particles.Lookat
    WriteAlignment=false
    WriteFacing=true
   )
   &DynamicMaterialParameters(
    Index 0 Param 2=&FloatFromCurve(
     FloatCurve=DataInterface() {propertyValues {Curve {keys [
         {interpMode=RCIM_Cubic, tangentMode=RCTM_User, time=0.5, value=0}
         {arriveTangent=6.0243, interpMode=RCIM_Cubic, leaveTangent=6.0243, tangentMode=RCTM_User, time=1, value=1}
        ]}}}
     Scale Curve=2
    )
    Index 0 Param 3=&FloatFromCurve(
     FloatCurve=DataInterface() {propertyValues {Curve {keys [
         {interpMode=RCIM_Cubic, tangentMode=RCTM_User, time=0.5, value=0}
         {arriveTangent=4.3588, interpMode=RCIM_Cubic, leaveTangent=4.3588, tangentMode=RCTM_User, time=1, value=1}
        ]}}}
     Scale Curve=0.3
    )
    Index 0 Param 4=&RandomRangeFloat(Minimum=0, Maximum=1)
    Index 1 Param 2=1
    Index 1 Param 3=&RandomRangeFloat(Minimum=0.1, Maximum=1)
    Module.Index0Param0Write=true
    Module.Index0Param1Write=true
    Module.Index0Param2Write=true
    Module.Index0Param3Write=true
    Module.Index1Param0Write=true
    Module.Index1Param1Write=true
    Module.Index1Param2Write=true
    Module.Index1Param3Write=false
    Module.Index2Param0Write=true
    Module.Index2Param1Write=true
    Module.Index2Param2Write=true
    Module.Index2Param3Write=true
    Module.Index3Param0Write=true
    Module.Index3Param1Write=true
    Module.Index3Param2Write=true
    Module.Index3Param3Write=true
    Write Parameter Index 1=true
    Write Parameter Index 2=true
    Write Parameter Index 3=true
   )
   &ScaleColor(
    Scale Alpha=&FloatFromCurve(
     FloatCurve=DataInterface() {propertyValues {Curve {keys [
         {arriveTangent=9.0725, interpMode=RCIM_Cubic, leaveTangent=9.0725, tangentMode=RCTM_User, time=0, value=0}
         {arriveTangent=0.0337, interpMode=RCIM_Cubic, leaveTangent=0.0337, tangentMode=RCTM_User, time=0.1408, value=0.9847}
         {arriveTangent=0.0046, interpMode=RCIM_Cubic, leaveTangent=0.0046, tangentMode=RCTM_User, time=0.75, value=1}
         {arriveTangent=-0.0405, interpMode=RCIM_Cubic, leaveTangent=-0.0405, tangentMode=RCTM_User, time=0.9, value=0}
        ]}}}
     Scale Curve=&RandomRangeFloat(Minimum=0.25, Maximum=1)
    )
    Scale RGB=&VectorFromFloat(Value=&RandomRangeFloat(Minimum=1, Maximum=20))
    ScaleA=true
    ScaleRGB=false
   )
  ]
  rendererClasses [&NiagaraSpriteRendererProperties]
  renderers [{rendererClass=&NiagaraSpriteRendererProperties}]
 }
 SmokePuffsBody {
  emitterUpdateScript [
   &EmitterState(Life Cycle Mode=::NewEnumerator1, Loop Behavior=::NewEnumerator1, Loop Duration=1.25, UseLoopDelay=false)
   &SpawnRate(SpawnRate=10, Use Spawn Probability=false)
  ]
  particleSpawnScript [
   &InitializeParticle(
    Color Maximum=<rgba 0.2193 0.1719 1 1>
    Color Minimum=<rgba 0.0865 0.3008 1 1>
    Color Mode=::NewEnumerator2
    Lifetime=&RandomRangeFloat(Minimum=3, Maximum=3.2)
    Mesh Scale Mode=::NewEnumerator1
    Mesh Uniform Scale Max=30
    Mesh Uniform Scale Min=10
    Position Offset=$System.SpiralPositionOffset
    Sprite Rotation Mode=::NewEnumerator2
    Sprite Size Mode=::NewEnumerator3
    Uniform Sprite Size=&RemapRange(
     Input Value To Remap=&FloatFromCurve(
      FloatCurve=DataInterface() {propertyValues {Curve {keys [
          {arriveTangent=0.3815, interpMode=RCIM_Cubic, leaveTangent=0.3815, tangentMode=RCTM_User, time=0, value=0}
          {arriveTangent=3.6818, interpMode=RCIM_Cubic, leaveTangent=3.6818, tangentMode=RCTM_User, time=1, value=1}
         ]}}}
      CurveIndex=&RandomRangeFloat(Minimum=0, Maximum=1)
     )
     Input Min=0
     Input Max=1
     Output Min=10
     Output Max=20
    )
    UsePositionOffset=true
    Write Material Random=true
   )
   &KillParticles(Kill Particles=&SetBoolByFloatComparison(A=&RandomRangeFloat(Minimum=0, Maximum=1), B=0.9))
   &CameraOffset(Camera Offset Amount=25) {enabled=false}
  ]
  particleUpdateScript [
   &ParticleState()
   /SetVariables(Particles.SpiralRadius:NiagaraFloat=&Multiply_Float(
     A=$System.SpiralRadius
     B=&FloatFromCurve(FloatCurve=DataInterface() {propertyValues {Curve {keys [
          {arriveTangent=0.0323, interpMode=RCIM_Cubic, leaveTangent=0.0323, tangentMode=RCTM_User, time=0, value=0.4}
          {arriveTangent=3.0767, interpMode=RCIM_Cubic, leaveTangent=3.0767, tangentMode=RCTM_User, time=0.5, value=0.8}
          {arriveTangent=0.0337, interpMode=RCIM_Cubic, leaveTangent=0.0337, tangentMode=RCTM_User, time=0.6, value=1}
          {arriveTangent=-3.8238, interpMode=RCIM_Cubic, leaveTangent=-3.8238, tangentMode=RCTM_User, time=0.7, value=0.8}
          <at 1 0.275 RCIM_Cubic>
         ]}}})
    )) {isSetParameters=true, module=SetVariables_#7}
   /SetVariables(Particles.SpiralSpeed:NiagaraFloat=&FloatFromCurve(FloatCurve=DataInterface() {propertyValues {Curve {keys [
         {interpMode=RCIM_Cubic, tangentMode=RCTM_User, time=0.5, value=1}
         {arriveTangent=-0.9145, interpMode=RCIM_Cubic, leaveTangent=-0.9145, tangentMode=RCTM_User, time=1, value=0.7}
        ]}}})) {isSetParameters=true, module=SetVariables_#8}
   /SetVariables(Particles.Position=&AddVectorToPosition(
     Position=$Particles.Initial.Position
     Vector=&MakeVector(
      X=&Sine(
       Normalized Angle=$Particles.Age
       Period=&Multiply_Float(A=&Multiply_Float(A=$Particles.Lifetime, B=$Particles.SpiralSpeed), B=$System.SpiralRotations)
       Scale=$Particles.SpiralRadius
       Bias=0
      )
      Y=&Cosine(
       Normalized Angle=$Particles.Age
       Period=&Multiply_Float(A=&Multiply_Float(A=$Particles.Lifetime, B=$Particles.SpiralSpeed), B=$System.SpiralRotations)
       Scale=$Particles.SpiralRadius
       Bias=0
      )
      Z=&Lerp_Float(
       A=0
       B=$System.SpiralHeight
       Alpha=&FloatFromCurve(
        FloatCurve=DataInterface() {propertyValues {Curve {keys [{tangentMode=RCTM_User, time=0, value=0}, {tangentMode=RCTM_User, time=0.7, value=1}]}}}
        Scale Curve=&RandomRangeFloat(Minimum=1, Maximum=1.1)
       )
      )
     )
    )) {isSetParameters=true, module=SetVariables_#9}
   &ScaleSpriteSize(
    Uniform Curve Index=&ReturnNormalizedExecIndex()
    Uniform Curve Scale=2.5
    Uniform Curve Sprite Scale=DataInterface() {propertyValues {Curve {keys [{interpMode=RCIM_Cubic, tangentMode=RCTM_User, time=0, value=0}, {interpMode=RCIM_Cubic, tangentMode=RCTM_User, time=0.5, value=1}]}}}
   ) {module=ScaleSpriteSize001}
   &ScaleColor(
    Scale Alpha=&FloatFromCurve(FloatCurve=DataInterface() {propertyValues {Curve {keys [
         {interpMode=RCIM_Cubic, tangentMode=RCTM_User, time=0, value=0}
         {interpMode=RCIM_Cubic, tangentMode=RCTM_User, time=0.2, value=1}
         {arriveTangent=0.0015, interpMode=RCIM_Cubic, leaveTangent=0.0015, tangentMode=RCTM_User, time=0.7, value=1}
         {tangentMode=RCTM_User, time=1, value=0}
        ]}}})
    ScaleA=true
    ScaleRGB=false
   ) {module=ScaleColor001}
   &ScaleColor(
    Scale Alpha=&FloatFromCurve(
     FloatCurve=DataInterface() {propertyValues {Curve {keys [
         {interpMode=RCIM_Cubic, tangentMode=RCTM_User, time=0.1, value=0}
         {interpMode=RCIM_Cubic, tangentMode=RCTM_User, time=0.4, value=1}
         {arriveTangent=0.0015, interpMode=RCIM_Cubic, leaveTangent=0.0015, tangentMode=RCTM_User, time=0.8, value=1}
         {tangentMode=RCTM_User, time=1, value=0}
        ]}}}
     CurveIndex=&ReturnNormalizedExecIndex()
     Scale Curve=0.25
    )
    ScaleA=true
    ScaleRGB=false
   )
   &ScaleColor(
    Scale Alpha=&FloatFromCurve(
     FloatCurve=DataInterface() {propertyValues {Curve {keys [
         {arriveTangent=-0.0405, interpMode=RCIM_Cubic, leaveTangent=-0.0405, tangentMode=RCTM_User, time=2.4, value=1}
         {arriveTangent=-0.7773, interpMode=RCIM_Cubic, leaveTangent=-0.7773, tangentMode=RCTM_User, time=5, value=0}
        ]}}}
     CurveIndex=$Emitter.Age
     Scale Curve=4
    )
    ScaleA=true
    ScaleRGB=false
   ) {module=ScaleColor003}
   &DynamicMaterialParameters(
    Index 0 Param 1=&Multiply_Float(
     A=&Add_Float(A=$Engine.RealTime, B=&RandomRangeFloat(Minimum=0, Maximum=10))
     B=0.1
    )
    Index 0 Param 2=&FloatFromCurve(FloatCurve=DataInterface() {propertyValues {Curve {keys [
         {interpMode=RCIM_Cubic, tangentMode=RCTM_User, time=0, value=1}
         {interpMode=RCIM_Cubic, tangentMode=RCTM_User, time=0.2, value=0.75}
         {arriveTangent=0.0015, interpMode=RCIM_Cubic, leaveTangent=0.0015, tangentMode=RCTM_User, time=0.7, value=0.75}
         {tangentMode=RCTM_User, time=1, value=1}
        ]}}})
    Index 2 Param 1=&FloatFromCurve(FloatCurve=DataInterface() {propertyValues {Curve {keys [
         {interpMode=RCIM_Cubic, tangentMode=RCTM_User, time=0, value=0}
         {arriveTangent=-0.1028, interpMode=RCIM_Cubic, leaveTangent=-0.1028, tangentMode=RCTM_User, time=1, value=2}
        ]}}})
    Index 2 Param 2=$Particles.MaterialRandom
    Index 2 Param 3=&FloatFromCurve(FloatCurve=DataInterface() {propertyValues {Curve {keys [
         {interpMode=RCIM_Cubic, tangentMode=RCTM_User, time=0, value=0}
         {arriveTangent=4.4069, interpMode=RCIM_Cubic, leaveTangent=4.4069, tangentMode=RCTM_User, time=1, value=2}
        ]}}})
    Index 2 Param 4=0.75
    Module.Index0Param0Write=true
    Module.Index0Param1Write=true
    Module.Index0Param2Write=true
    Module.Index0Param3Write=true
    Module.Index2Param0Write=true
    Module.Index2Param1Write=true
    Module.Index2Param2Write=true
    Module.Index2Param3Write=true
    Write Parameter Index 2=true
   )
   /SetVariables(Particles.Velocity=&Multiply_VectorByFloat(
     Vector=&SubtractVectorFromPosition(Position=$Particles.Position, Vector=Linked() {linkedVariable {name=Particles.Previous.Position, type=NiagaraPosition}})
     Float=$Engine.DeltaTime
    )) {isSetParameters=true, module=SetVariables_#10}
   &ScaleSpriteSizeBySpeed(Max Scale Factor=<1 250>, Velocity Threshold=300)
  ]
  rendererClasses [&NiagaraSpriteRendererProperties]
  renderers [
   {rendererClass=&NiagaraSpriteRendererProperties}
   {rendererClass=&NiagaraSpriteRendererProperties, rendererIndex=1}
  ]
 }
 SmokePuffsBodyGPU {
  simTarget=GPUComputeSim
  emitterUpdateScript [
   &EmitterState(Life Cycle Mode=::NewEnumerator1, Loop Behavior=::NewEnumerator1, Loop Duration=1.25, UseLoopDelay=false)
   &SpawnRate(SpawnRate=30, Use Spawn Probability=false)
  ]
  particleSpawnScript [
   &InitializeParticle(
    Color Maximum=<rgba 0.2193 0.1719 1 1>
    Color Minimum=<rgba 0.0865 0.3008 1 1>
    Color Mode=::NewEnumerator2
    Lifetime=&RandomRangeFloat(Minimum=3, Maximum=3.2)
    Mesh Scale Mode=::NewEnumerator1
    Mesh Uniform Scale Max=30
    Mesh Uniform Scale Min=10
    Position Offset=$System.SpiralPositionOffset
    Sprite Rotation Mode=::NewEnumerator2
    Sprite Size Mode=::NewEnumerator3
    Uniform Sprite Size=&RemapRange(
     Input Value To Remap=&FloatFromCurve(
      FloatCurve=DataInterface() {propertyValues {Curve {keys [
          {arriveTangent=0.3815, interpMode=RCIM_Cubic, leaveTangent=0.3815, tangentMode=RCTM_User, time=0, value=0}
          {arriveTangent=3.6818, interpMode=RCIM_Cubic, leaveTangent=3.6818, tangentMode=RCTM_User, time=1, value=1}
         ]}}}
      CurveIndex=&RandomRangeFloat(Minimum=0, Maximum=1)
     )
     Input Min=0
     Input Max=1
     Output Min=10
     Output Max=20
    )
    UsePositionOffset=true
    Write Material Random=true
   )
   &KillParticles(Kill Particles=&SetBoolByFloatComparison(A=&RandomRangeFloat(Minimum=0, Maximum=1), B=0.9))
   &CameraOffset(Camera Offset Amount=25) {enabled=false}
  ]
  particleUpdateScript [
   &ParticleState()
   /SetVariables(Particles.SpiralRadius:NiagaraFloat=&Multiply_Float(
     A=$System.SpiralRadius
     B=&FloatFromCurve(FloatCurve=DataInterface() {propertyValues {Curve {keys [
          {arriveTangent=0.0323, interpMode=RCIM_Cubic, leaveTangent=0.0323, tangentMode=RCTM_User, time=0, value=0.4}
          {arriveTangent=3.0767, interpMode=RCIM_Cubic, leaveTangent=3.0767, tangentMode=RCTM_User, time=0.5, value=0.8}
          {arriveTangent=0.0337, interpMode=RCIM_Cubic, leaveTangent=0.0337, tangentMode=RCTM_User, time=0.6, value=1}
          {arriveTangent=-3.8238, interpMode=RCIM_Cubic, leaveTangent=-3.8238, tangentMode=RCTM_User, time=0.7, value=0.8}
          <at 1 0.275 RCIM_Cubic>
         ]}}})
    )) {isSetParameters=true, module=SetVariables_#7}
   /SetVariables(Particles.SpiralSpeed:NiagaraFloat=&FloatFromCurve(FloatCurve=DataInterface() {propertyValues {Curve {keys [
         {interpMode=RCIM_Cubic, tangentMode=RCTM_User, time=0.5, value=1}
         {arriveTangent=-0.9145, interpMode=RCIM_Cubic, leaveTangent=-0.9145, tangentMode=RCTM_User, time=1, value=0.7}
        ]}}})) {isSetParameters=true, module=SetVariables_#8}
   /SetVariables(Particles.Position=&AddVectorToPosition(
     Position=$Particles.Initial.Position
     Vector=&MakeVector(
      X=&Sine(
       Normalized Angle=$Particles.Age
       Period=&Multiply_Float(A=&Multiply_Float(A=$Particles.Lifetime, B=$Particles.SpiralSpeed), B=$System.SpiralRotations)
       Scale=$Particles.SpiralRadius
       Bias=0
      )
      Y=&Cosine(
       Normalized Angle=$Particles.Age
       Period=&Multiply_Float(A=&Multiply_Float(A=$Particles.Lifetime, B=$Particles.SpiralSpeed), B=$System.SpiralRotations)
       Scale=$Particles.SpiralRadius
       Bias=0
      )
      Z=&Lerp_Float(
       A=0
       B=$System.SpiralHeight
       Alpha=&FloatFromCurve(
        FloatCurve=DataInterface() {propertyValues {Curve {keys [{tangentMode=RCTM_User, time=0, value=0}, {tangentMode=RCTM_User, time=0.7, value=1}]}}}
        Scale Curve=&RandomRangeFloat(Minimum=1, Maximum=1.1)
       )
      )
     )
    )) {isSetParameters=true, module=SetVariables_#9}
   &ScaleSpriteSize(
    Uniform Curve Index=&ReturnNormalizedExecIndex()
    Uniform Curve Scale=2.5
    Uniform Curve Sprite Scale=DataInterface() {propertyValues {Curve {keys [{interpMode=RCIM_Cubic, tangentMode=RCTM_User, time=0, value=0}, {interpMode=RCIM_Cubic, tangentMode=RCTM_User, time=0.5, value=1}]}}}
   ) {module=ScaleSpriteSize001}
   &ScaleColor(
    Scale Alpha=&FloatFromCurve(FloatCurve=DataInterface() {propertyValues {Curve {keys [
         {interpMode=RCIM_Cubic, tangentMode=RCTM_User, time=0, value=0}
         {interpMode=RCIM_Cubic, tangentMode=RCTM_User, time=0.2, value=1}
         {arriveTangent=0.0015, interpMode=RCIM_Cubic, leaveTangent=0.0015, tangentMode=RCTM_User, time=0.7, value=1}
         {tangentMode=RCTM_User, time=1, value=0}
        ]}}})
    ScaleA=true
    ScaleRGB=false
   ) {module=ScaleColor001}
   &ScaleColor(
    Scale Alpha=&FloatFromCurve(
     FloatCurve=DataInterface() {propertyValues {Curve {keys [
         {interpMode=RCIM_Cubic, tangentMode=RCTM_User, time=0.1, value=0}
         {interpMode=RCIM_Cubic, tangentMode=RCTM_User, time=0.4, value=1}
         {arriveTangent=0.0015, interpMode=RCIM_Cubic, leaveTangent=0.0015, tangentMode=RCTM_User, time=0.8, value=1}
         {tangentMode=RCTM_User, time=1, value=0}
        ]}}}
     CurveIndex=&ReturnNormalizedExecIndex()
     Scale Curve=0.25
    )
    ScaleA=true
    ScaleRGB=false
   )
   &ScaleColor(
    Scale Alpha=&FloatFromCurve(
     FloatCurve=DataInterface() {propertyValues {Curve {keys [
         {arriveTangent=-0.0405, interpMode=RCIM_Cubic, leaveTangent=-0.0405, tangentMode=RCTM_User, time=2.4, value=1}
         {arriveTangent=-0.7773, interpMode=RCIM_Cubic, leaveTangent=-0.7773, tangentMode=RCTM_User, time=5, value=0}
        ]}}}
     CurveIndex=$Emitter.Age
     Scale Curve=4
    )
    ScaleA=true
    ScaleRGB=false
   ) {module=ScaleColor003}
   &DynamicMaterialParameters(
    Index 0 Param 1=&Multiply_Float(
     A=&Add_Float(A=$Engine.RealTime, B=&RandomRangeFloat(Minimum=0, Maximum=10))
     B=0.1
    )
    Index 0 Param 2=&FloatFromCurve(FloatCurve=DataInterface() {propertyValues {Curve {keys [
         {interpMode=RCIM_Cubic, tangentMode=RCTM_User, time=0, value=1}
         {interpMode=RCIM_Cubic, tangentMode=RCTM_User, time=0.2, value=0.75}
         {arriveTangent=0.0015, interpMode=RCIM_Cubic, leaveTangent=0.0015, tangentMode=RCTM_User, time=0.7, value=0.75}
         {tangentMode=RCTM_User, time=1, value=1}
        ]}}})
    Index 2 Param 1=&FloatFromCurve(FloatCurve=DataInterface() {propertyValues {Curve {keys [
         {interpMode=RCIM_Cubic, tangentMode=RCTM_User, time=0, value=0}
         {arriveTangent=-0.1028, interpMode=RCIM_Cubic, leaveTangent=-0.1028, tangentMode=RCTM_User, time=1, value=2}
        ]}}})
    Index 2 Param 2=$Particles.MaterialRandom
    Index 2 Param 3=&FloatFromCurve(FloatCurve=DataInterface() {propertyValues {Curve {keys [
         {interpMode=RCIM_Cubic, tangentMode=RCTM_User, time=0, value=0}
         {arriveTangent=4.4069, interpMode=RCIM_Cubic, leaveTangent=4.4069, tangentMode=RCTM_User, time=1, value=2}
        ]}}})
    Index 2 Param 4=0.75
    Module.Index0Param0Write=true
    Module.Index0Param1Write=true
    Module.Index0Param2Write=true
    Module.Index0Param3Write=true
    Module.Index2Param0Write=true
    Module.Index2Param1Write=true
    Module.Index2Param2Write=true
    Module.Index2Param3Write=true
    Write Parameter Index 2=true
   )
   /SetVariables(Particles.Velocity=&Multiply_VectorByFloat(
     Vector=&SubtractVectorFromPosition(Position=$Particles.Position, Vector=Linked() {linkedVariable {name=Particles.Previous.Position, type=NiagaraPosition}})
     Float=$Engine.DeltaTime
    )) {isSetParameters=true, module=SetVariables_#10}
   &ScaleSpriteSizeBySpeed(Max Scale Factor=<1 250>, Velocity Threshold=300)
  ]
  rendererClasses [&NiagaraSpriteRendererProperties]
  renderers [
   {rendererClass=&NiagaraSpriteRendererProperties}
   {rendererClass=&NiagaraSpriteRendererProperties, rendererIndex=1}
  ]
 }
 SmokeTrailBody {
  emitterUpdateScript [
   &EmitterState(Life Cycle Mode=::NewEnumerator1, Loop Behavior=::NewEnumerator1, Loop Duration=2.25, UseLoopDelay=false)
   &SpawnRate(SpawnRate=12, Use Spawn Probability=false)
  ]
  particleSpawnScript [
   &InitializeParticle(
    Color=<rgba 0.0865 0.3008 1 1>
    Lifetime=3
    Position Offset=$System.SpiralPositionOffset
    Ribbon Facing Mode=::NewEnumerator1
    "Ribbon Twist "=1
    Ribbon Twist Mode:ENiagara_UnsetDirectSet=::NewEnumerator1
    Ribbon Width=1
    Ribbon Width Mode=::NewEnumerator1
    Sprite Size Mode=::NewEnumerator3
    Uniform Sprite Size=5
    UsePositionOffset=true
    Write Material Random=true
   )
   /SetVariables(Particles.RibbonFacing=<0 0 1>) {isSetParameters=true}
  ]
  particleUpdateScript [
   &ParticleState()
   /SetVariables(Particles.SpiralRadius:NiagaraFloat=&Multiply_Float(
     A=$System.SpiralRadius
     B=&FloatFromCurve(FloatCurve=DataInterface() {propertyValues {Curve {keys [
          {arriveTangent=0.0323, interpMode=RCIM_Cubic, leaveTangent=0.0323, tangentMode=RCTM_User, time=0, value=0.4}
          {arriveTangent=3.0767, interpMode=RCIM_Cubic, leaveTangent=3.0767, tangentMode=RCTM_User, time=0.5, value=0.8}
          {arriveTangent=0.0337, interpMode=RCIM_Cubic, leaveTangent=0.0337, tangentMode=RCTM_User, time=0.6, value=1}
          {arriveTangent=-3.8238, interpMode=RCIM_Cubic, leaveTangent=-3.8238, tangentMode=RCTM_User, time=0.7, value=0.8}
          <at 1 0.275 RCIM_Cubic>
         ]}}})
    )) {isSetParameters=true}
   /SetVariables(Particles.SpiralSpeed:NiagaraFloat=&FloatFromCurve(FloatCurve=DataInterface() {propertyValues {Curve {keys [
         {interpMode=RCIM_Cubic, tangentMode=RCTM_User, time=0.5, value=1}
         {arriveTangent=-0.9145, interpMode=RCIM_Cubic, leaveTangent=-0.9145, tangentMode=RCTM_User, time=1, value=0.7}
        ]}}})) {isSetParameters=true}
   /SetVariables(Particles.Position=&AddVectorToPosition(
     Position=$Particles.Initial.Position
     Vector=&MakeVector(
      X=&Sine(
       Normalized Angle=$Particles.Age
       Period=&Multiply_Float(A=&Multiply_Float(A=$Particles.Lifetime, B=$Particles.SpiralSpeed), B=$System.SpiralRotations)
       Scale=$Particles.SpiralRadius
       Bias=0
      )
      Y=&Cosine(
       Normalized Angle=$Particles.Age
       Period=&Multiply_Float(A=&Multiply_Float(A=$Particles.Lifetime, B=$Particles.SpiralSpeed), B=$System.SpiralRotations)
       Scale=$Particles.SpiralRadius
       Bias=0
      )
      Z=&Lerp_Float(
       A=0
       B=$System.SpiralHeight
       Alpha=&FloatFromCurve(FloatCurve=DataInterface() {propertyValues {Curve {keys [{tangentMode=RCTM_User, time=0, value=0}, {tangentMode=RCTM_User, time=0.7, value=1}]}}})
      )
     )
    )) {isSetParameters=true}
   &ScaleRibbonWidth(Ribbon Width Scale=&FloatFromCurve(FloatCurve=DataInterface() {propertyValues {Curve {keys [{interpMode=RCIM_Cubic, tangentMode=RCTM_User, time=0, value=0}, {interpMode=RCIM_Cubic, tangentMode=RCTM_User, time=0.5, value=1}]}}}))
   &ScaleRibbonWidth(Ribbon Width Scale=&FloatFromCurve(
     FloatCurve=DataInterface() {propertyValues {Curve {keys [
         {interpMode=RCIM_Cubic, tangentMode=RCTM_User, time=0, value=0}
         {arriveTangent=1.4286, interpMode=RCIM_Cubic, leaveTangent=1.4286, time=0.3, value=1}
         {arriveTangent=-1.4286, interpMode=RCIM_Cubic, leaveTangent=-1.4286, time=0.7, value=1}
         {tangentMode=RCTM_User, time=1, value=0}
        ]}}}
     CurveIndex=&ReturnNormalizedExecIndex()
     Scale Curve=70
    )) {module=ScaleRibbonWidth001}
   &ScaleColor(
    Scale Alpha=&FloatFromCurve(
     FloatCurve=DataInterface() {propertyValues {Curve {keys [
         {interpMode=RCIM_Cubic, tangentMode=RCTM_User, time=0.1, value=1}
         {interpMode=RCIM_Cubic, tangentMode=RCTM_User, time=0.3, value=1}
         {arriveTangent=0.0015, interpMode=RCIM_Cubic, leaveTangent=0.0015, tangentMode=RCTM_User, time=0.9, value=1}
         {tangentMode=RCTM_User, time=1, value=0}
        ]}}}
     CurveIndex=&ReturnNormalizedExecIndex()
     Scale Curve=0.5
    )
    ScaleA=true
    ScaleRGB=false
   )
   &DynamicMaterialParameters(
    Index 0 Param 1=&Multiply_Float(A=$Engine.RealTime, B=1)
    Index 0 Param 2=&Multiply_Float(A=$Engine.RealTime, B=1)
    Index 0 Param 3=10
    Index 0 Param 4=&RandomRangeFloat(Minimum=0.2, Maximum=0.4)
    Index 1 Param 1=100
    Module.Index0Param0Write=true
    Module.Index0Param1Write=true
    Module.Index0Param2Write=true
    Module.Index0Param3Write=true
    Module.Index1Param0Write=true
    Module.Index1Param1Write=true
    Module.Index1Param2Write=true
    Module.Index1Param3Write=true
    Write Parameter Index 1=true
   )
   &ScaleColor(
    Scale Alpha=&FloatFromCurve(
     FloatCurve=DataInterface() {propertyValues {Curve {keys [
         <at 0 0 RCIM_Cubic>
         {arriveTangent=0.4167, interpMode=RCIM_Cubic, leaveTangent=0.4167, time=0.7, value=1}
         {arriveTangent=-0.0405, interpMode=RCIM_Cubic, leaveTangent=-0.0405, tangentMode=RCTM_User, time=2.4, value=1}
         {arriveTangent=-0.7773, interpMode=RCIM_Cubic, leaveTangent=-0.7773, tangentMode=RCTM_User, time=4, value=0}
        ]}}}
     CurveIndex=$Emitter.Age
     Scale Curve=4
    )
    ScaleA=true
    ScaleRGB=false
   ) {module=ScaleColor001}
  ]
  rendererClasses [&NiagaraRibbonRendererProperties]
  renderers [{rendererClass=&NiagaraRibbonRendererProperties}]
 }
 WindBase {
  emitterUpdateScript [
   &EmitterState(Life Cycle Mode=::NewEnumerator1, Loop Behavior=::NewEnumerator1, UseLoopDelay=false)
   &SpawnBurst_Instantaneous(Use Spawn Probability=false, UseLoopCountLimit=false) {enabled=false}
   &SpawnBurst_Instantaneous(Spawn Group=1, Use Spawn Probability=false, UseLoopCountLimit=false) {module=SpawnBurst_Instantaneous001}
  ]
  particleSpawnScript [
   &InitializeParticle(
    Color=&MaskLinearColorBySpawnGroup(Masked Linear Color=<rgba 0.2193 0.1719 1 1>, Passthrough Linear Color=<rgba 0.0865 0.3008 1 1>)
    Lifetime=4
    Mesh Scale Mode=::NewEnumerator3
    Mesh Uniform Scale=&Divide_Float(A=$User.Radius, B=&MaskFloatBySpawnGroup(Masked Float=150, Passthrough Float=100))
    Position Offset=<0 0 -40>
    UsePositionOffset=true
    Write Material Random=true
   )
   &InitialMeshOrientation(
    Mesh Orientation Mode=::NewEnumerator2
    Orientation Vector=<0 0 1>
    Rotation=&MakeVector(
     X=&Add_Float(
      A=&RandomRangeFloat(Minimum=-0.1, Maximum=0.1)
      B=&MaskFloatBySpawnGroup(Masked Float=0, Passthrough Float=0.5)
     )
     Y=0
     Z=0
    )
    Use Rotation Vector=true
   )
   &CameraOffset(Camera Offset Amount=20)
  ]
  particleUpdateScript [
   &ParticleState()
   &ScaleMeshSize(Scale Factor=&Multiply_VectorByFloat(Vector=<2 1 1>, Float=1))
   &ScaleColor(
    Scale Alpha=&FloatFromCurve(FloatCurve=DataInterface() {propertyValues {Curve {keys [
         {interpMode=RCIM_Cubic, tangentMode=RCTM_User, time=0, value=0}
         {interpMode=RCIM_Cubic, tangentMode=RCTM_User, time=0.25, value=1}
         {tangentMode=RCTM_User, time=1, value=0}
        ]}}})
    Scale RGB=&VectorFromFloat(Value=2)
    ScaleA=true
    ScaleRGB=true
   )
   &DynamicMaterialParameters(
    Index 0 Param 1=$Engine.RealTime
    Index 0 Param 2=&MaskFloatBySpawnGroup(Masked Float=0.25, Passthrough Float=0.15)
    Module.Index0Param0Write=true
    Module.Index0Param1Write=true
    Module.Index0Param2Write=false
    Module.Index0Param3Write=false
   )
   &FN_Light_Attributes(
    Color=<rgba 0.0865 0.3008 1 1>
    Create Color Attribute=true
    Create Exponent Attribute=true
    Create Position Attribute=true
    Create Radius Attribute=true
    Position=&AddVectorToPosition(Position=$Particles.Position, Vector=<0 0 50>)
    Radius=200
    Scale Color=&FloatFromCurve(
     FloatCurve=DataInterface() {propertyValues {Curve {keys [
         {interpMode=RCIM_Cubic, tangentMode=RCTM_User, time=0, value=0}
         {interpMode=RCIM_Cubic, tangentMode=RCTM_User, time=0.4, value=1}
         {tangentMode=RCTM_User, time=1, value=0}
        ]}}}
     Scale Curve=50
    )
   )
  ]
  rendererClasses [&NiagaraMeshRendererProperties, &NiagaraLightRendererProperties]
  renderers [
   {rendererClass=&NiagaraMeshRendererProperties, rendererIndex=0}
   {rendererClass=&NiagaraLightRendererProperties, rendererIndex=1}
  ]
 }
 WindFunnel {
  emitterUpdateScript [
   &EmitterState(Life Cycle Mode=::NewEnumerator1, Loop Behavior=::NewEnumerator1, Loop Delay=0.2, UseLoopDelay=true)
   &SpawnBurst_Instantaneous(Use Spawn Probability=false, UseLoopCountLimit=false) {enabled=false}
   &SpawnBurst_Instantaneous(Spawn Group=1, Use Spawn Probability=false, UseLoopCountLimit=false) {module=SpawnBurst_Instantaneous001}
  ]
  particleSpawnScript [
   &InitializeParticle(
    Color=&MaskLinearColorBySpawnGroup(Masked Linear Color=<rgba 0.2193 0.1719 1 1>, Passthrough Linear Color=<rgba 0.0865 0.3008 1 1>)
    Lifetime=3.5
    Mesh Scale Mode=::NewEnumerator3
    Mesh Uniform Scale=&Divide_Float(A=400, B=&MaskFloatBySpawnGroup(Masked Float=150, Passthrough Float=100))
    Position Offset=<0 0 -40>
    UsePositionOffset=true
    Write Material Random=true
   )
   &CameraOffset(Camera Offset Amount=20)
  ]
  particleUpdateScript [
   &ParticleState()
   /SetVariables(Particles.Position=&AddVectorToPosition(
     Position=$Particles.Initial.Position
     Vector=&MakeVector(
      X=0
      Y=0
      Z=&FloatFromCurve(FloatCurve=DataInterface() {propertyValues {Curve {keys [
           {interpMode=RCIM_Cubic, tangentMode=RCTM_User, time=0, value=-80}
           {arriveTangent=69.8376, interpMode=RCIM_Cubic, leaveTangent=69.8376, tangentMode=RCTM_User, time=1, value=0}
          ]}}})
     )
    )) {isSetParameters=true}
   &ScaleColor(
    Scale Alpha=&FloatFromCurve(
     FloatCurve=DataInterface() {propertyValues {Curve {keys [
         {interpMode=RCIM_Cubic, tangentMode=RCTM_User, time=0, value=0}
         {interpMode=RCIM_Cubic, tangentMode=RCTM_User, time=0.5, value=0.95}
         {tangentMode=RCTM_User, time=1, value=0}
        ]}}}
     Scale Curve=2
    )
    Scale RGB=&VectorFromFloat(Value=2)
    ScaleA=true
    ScaleRGB=true
   )
   &DynamicMaterialParameters(
    Index 0 Param 1=&FloatFromCurve(FloatCurve=DataInterface() {propertyValues {Curve {keys [
         {interpMode=RCIM_Cubic, tangentMode=RCTM_User, time=0, value=0}
         {arriveTangent=1.4134, interpMode=RCIM_Cubic, leaveTangent=1.4134, tangentMode=RCTM_User, time=1, value=1}
        ]}}})
    Index 0 Param 2=&MaskFloatBySpawnGroup(Masked Float=0.5, Passthrough Float=0.3)
    Module.Index0Param0Write=true
    Module.Index0Param1Write=true
    Module.Index0Param2Write=false
    Module.Index0Param3Write=false
    Module.Index1Param0Write=true
    Module.Index1Param1Write=true
    Module.Index1Param2Write=true
    Module.Index1Param3Write=true
    Module.Index2Param0Write=true
    Module.Index2Param1Write=true
    Module.Index2Param2Write=true
    Module.Index2Param3Write=true
    Module.Index3Param0Write=true
    Module.Index3Param1Write=true
    Module.Index3Param2Write=true
    Module.Index3Param3Write=true
    Write Parameter Index 1=true
    Write Parameter Index 2=true
    Write Parameter Index 3=true
   )
  ]
  rendererClasses [&NiagaraMeshRendererProperties]
  renderers [{rendererClass=&NiagaraMeshRendererProperties, rendererIndex=0}]
 }
}
summary {
 systemName=NS_Emote_LemonCart_Granite_Trail
 emitters [
  {emitterName=SmokeTrailBody, rendererClasses [&NiagaraRibbonRendererProperties]}
  {emitterName=SmokePuffsBodyGPU, simTarget=GPUComputeSim, rendererClasses [&NiagaraSpriteRendererProperties]}
  {emitterName=WindBase, rendererClasses [&NiagaraMeshRendererProperties, &NiagaraLightRendererProperties]}
  {emitterName=DollarDollarBillYallGPU, simTarget=GPUComputeSim, rendererClasses [&NiagaraSpriteRendererProperties]}
  {emitterName=WindFunnel, rendererClasses [&NiagaraMeshRendererProperties]}
  {
   bEnabled=false
   emitterName=DEBUGGhostMtl
   rendererClasses [&NiagaraSpriteRendererProperties, &NiagaraMeshRendererProperties]
  }
  {emitterName=SmokePuffsBody, rendererClasses [&NiagaraSpriteRendererProperties]}
  {emitterName=DollarDollarBillYall, rendererClasses [&NiagaraSpriteRendererProperties]}
 ]
 userVariables [
  User.bIsFrontEnd {type=NiagaraBool}
  User.bIsFrontEndPreview {type=NiagaraBool}
  User.Duration {type=NiagaraFloat, defaultValue=5, description=""}
  User.Period {type=NiagaraFloat, defaultValue=1, description=""}
  User.Radius {type=NiagaraFloat, defaultValue=32, description=""}
  User.ScaleCurve {type=NiagaraFloat, defaultValue=50, description=""}
 ]
}
